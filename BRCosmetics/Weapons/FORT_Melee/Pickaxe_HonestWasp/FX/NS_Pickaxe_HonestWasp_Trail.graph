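# Niagara system that draws the swing trail ribbon for the HonestWasp pickaxe.
# A single emitter uses the FN anim trail modules to spawn, initialize and update a light grey, translucent ribbon.
# Particle lifetime alternates between two short values by the user swing section index.
# System-level fade values scale alpha at spawn and during update, and a time-of-day color tints the trail each update.
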NiagaraSystem /BRCosmetics/Weapons/FORT_Melee/Pickaxe_HonestWasp/FX/NS_Pickaxe_HonestWasp_Trail
dependencies {
 usedDataInterfaces [&NiagaraDataInterfaceSocketReader]
 usedDynamicInputs [
  &FNTimeOfDayColor
  &ClampFloat
  &RemapRange
  &Add_Float
  &Subtract_Float
 ]
 usedModules [
  &self:SystemSpawnScript.SystemScriptSource.SystemScriptGraph.NiagaraNodeAssignment_1.SetVariables_#1
  &self:SystemSpawnScript.SystemScriptSource.SystemScriptGraph.NiagaraNodeAssignment_3.SetVariables_#2
  &SystemState
  &self:SystemSpawnScript.SystemScriptSource.SystemScriptGraph.NiagaraNodeAssignment_9.SetVariables_#3
  &self:SystemSpawnScript.SystemScriptSource.SystemScriptGraph.NiagaraNodeAssignment_6.SetVariables_#4
 ]
}
emitters {TrailBase001 {
  emitterUpdateScript [
   &EmitterState()
   &FN_SpawnAnimTrail()
  ]
  particleSpawnScript [
   &FN_InitializeAnimTrailBoneIndecies()
   &FN_InitializeAnimTrail(
    Color=&MakeLinearColorFromVectorAndFloat(
     "Vector (RGB)"=&MakeVectorFromLinearColorRGB(Color=<rgba 0.45 0.45 0.45 1>)
     "Float (Alpha)"=&Multiply_Float(A=0.8, B=&Multiply_Float(A=$System.FadeSpawn, B=$System.FadeSpawnOut))
    )
    Lifetime=&MergeHelper_FloatSwitch(0=0.1, 1=0.14, 2=0.1, 3=0.14, Index=$User.SwingSectionIndex, Num=::NewEnumerator2)
   )
  ]
  particleUpdateScript [
   &FN_UpdateAnimTrail() {module=FN_UpdateAnimTrail001}
   &ScaleColor(
    Scale Mode=1
    Scale RGBA=&Multiply_Vector4(
     A=&MakeVector4_FromLinearColor(LinearColor=$System.TODColor)
     B=&MakeVector4FromVectorAndFloat("Vector (RGB)"=<1 1 1>, "Float (Alpha)"=$System.FadeSystem)
    )
    ScaleRGBA=true
   ) {module=ScaleColor002}
  ]
  rendererClasses [&NiagaraRibbonRendererProperties]
  renderers [{rendererClass=&NiagaraRibbonRendererProperties}]
 }}
summary {
 systemName=NS_Pickaxe_HonestWasp_Trail
 emitters [{emitterName=TrailBase001, rendererClasses [&NiagaraRibbonRendererProperties]}]
 userVariables [
  User.bIsFrontEnd {type=NiagaraBool}
  User.bIsFrontEndPreview {type=NiagaraBool}
  User.RibbonIDOffset {type=NiagaraInt32}
  User.SwingSectionIndex {type=NiagaraInt32}
 ]
}
