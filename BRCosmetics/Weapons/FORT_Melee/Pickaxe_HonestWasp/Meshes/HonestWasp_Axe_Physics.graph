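# Physics asset for the HonestWasp two-handed pickaxe, defining capsule bodies and joints for its dangling chains.
# One four-link chain hangs from base_chain, while arm and handle chains run from their own bases with limited swing and twist per link.
# The arm and handle chain ends join through an angularly free constraint, forming one span anchored at both ends.
# Unconstrained handle, right forearm, pelvis and right thigh capsules appear to serve as colliders.
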PhysicsAsset /BRCosmetics/Weapons/FORT_Melee/Pickaxe_HonestWasp/Meshes/HonestWasp_Axe_Physics
bodies {
 Melee_TwoHanded_Handle [{shapeType=Capsule, length=24.5, radius=2.5, shapeName=Melee_TwoHanded_Handle_capsule, center=<0 0 -4>}]
 base_chain [{shapeType=Capsule, length=0.505, radius=0.505, shapeName=base_chain_capsule, rotation=<rot 0 -90 0>}]
 base_chain2_arm [{shapeType=Capsule, length=0.505, radius=0.505, shapeName=base_chain2_arm_capsule, rotation=<rot 0 -90 0>}]
 base_chain2_handle [{shapeType=Capsule, length=0.505, radius=0.505, shapeName=base_chain2_handle_capsule, rotation=<rot 0 -90 0>}]
 dyn_chain2_arm_01 [{shapeType=Capsule, length=2.5, radius=1.5, shapeName=dyn_chain2_arm_01_capsule, center=<1.6 0 0>, rotation=<rot 90 0 0>}]
 dyn_chain2_arm_02 [{shapeType=Capsule, length=8.5, radius=1.5, shapeName=dyn_chain2_arm_02_capsule, center=<4.5 0 0>, rotation=<rot 90 0 0>}]
 dyn_chain2_arm_03 [{shapeType=Capsule, length=7, radius=1.5, shapeName=dyn_chain2_arm_03_capsule, center=<3.8 0 0>, rotation=<rot 90 0 0>}]
 dyn_chain2_arm_04 [{shapeType=Capsule, length=6.5, radius=1.5, shapeName=dyn_chain2_arm_04_capsule, center=<3.5 0 0>, rotation=<rot 90 0 0>}]
 dyn_chain2_arm_05 [{shapeType=Capsule, length=5.5, radius=2, shapeName=dyn_chain2_arm_05_capsule, center=<3.3 0 0>, rotation=<rot 90 0 0>}]
 dyn_chain2_handle_01 [{shapeType=Capsule, length=2, radius=1.5, shapeName=dyn_chain2_handle_01_capsule, center=<1 0 0>, rotation=<rot 90 0 0>}]
 dyn_chain2_handle_02 [{shapeType=Capsule, length=2, radius=1.8, shapeName=dyn_chain2_handle_02_capsule, center=<1.5 0 0>, rotation=<rot 90 0 0>}]
 dyn_chain2_handle_03 [{shapeType=Capsule, length=4, radius=2, shapeName=dyn_chain2_handle_03_capsule, center=<2.6 0 0>, rotation=<rot 90 0 0>}]
 dyn_chain2_handle_04 [{shapeType=Capsule, length=4, radius=2, shapeName=dyn_chain2_handle_04_capsule, center=<2.9 0 0>, rotation=<rot 90 0 0>}]
 dyn_chain_01 [{shapeType=Capsule, length=0.5, radius=1.5, shapeName=dyn_chain_01_capsule, center=<1.3746 0 0>, rotation=<rot 90 0 0>}]
 dyn_chain_02 [{shapeType=Capsule, length=3.5, radius=1.5, shapeName=dyn_chain_02_capsule, center=<3.317 0 0>, rotation=<rot 90 0 0>}]
 dyn_chain_03 [{shapeType=Capsule, length=2.5, radius=1.5, shapeName=dyn_chain_03_capsule, center=<3.317 0 0>, rotation=<rot 90 0 0>}]
 dyn_chain_04 [{shapeType=Capsule, length=2.5, radius=1.5, shapeName=dyn_chain_04_capsule, center=<3.317 0 0>, rotation=<rot 90 0 0>}]
 lowerarm_r [{shapeType=Capsule, length=22, radius=6, shapeName=lowerarm_r_capsule, center=<-12 0 0>, rotation=<rot 90 0 0>}]
 pelvis [{shapeType=Capsule, length=20, radius=11, shapeName=pelvis_capsule, rotation=<rot 0 -90 0>}]
 thigh_r [{shapeType=Capsule, length=33, radius=11, shapeName=thigh_l_capsule, center=<21.8542 0.5837 0.1317>, rotation=<rot 88.8195 172.6059 170.2331>}]
}
constraints [
 {bone1Name=dyn_chain_01, bone2Name=base_chain, swing1LimitDegrees=60, swing2LimitDegrees=5, twistLimitDegrees=1}
 {bone1Name=dyn_chain_02, bone2Name=dyn_chain_01, swing1LimitDegrees=5, swing2LimitDegrees=60, twistLimitDegrees=5}
 {bone1Name=dyn_chain_03, bone2Name=dyn_chain_02, swing1LimitDegrees=30, swing2LimitDegrees=30, twistLimitDegrees=10}
 {bone1Name=dyn_chain_04, bone2Name=dyn_chain_03, swing1LimitDegrees=45, swing2LimitDegrees=45, twistLimitDegrees=10}
 {bone1Name=dyn_chain2_arm_01, bone2Name=base_chain2_arm, swing1LimitDegrees=20, swing2LimitDegrees=20, twistLimitDegrees=1}
 {bone1Name=dyn_chain2_arm_02, bone2Name=dyn_chain2_arm_01, swing1LimitDegrees=25, swing2LimitDegrees=30, twistLimitDegrees=1}
 {bone1Name=dyn_chain2_arm_03, bone2Name=dyn_chain2_arm_02, swing1LimitDegrees=25, swing2LimitDegrees=30, twistLimitDegrees=2}
 {bone1Name=dyn_chain2_arm_04, bone2Name=dyn_chain2_arm_03, swing1LimitDegrees=20, swing2LimitDegrees=20, twistLimitDegrees=2}
 {bone1Name=dyn_chain2_arm_05, bone2Name=dyn_chain2_arm_04, swing1LimitDegrees=10, swing2LimitDegrees=10, twistLimitDegrees=1}
 {bone1Name=dyn_chain2_handle_01, bone2Name=base_chain2_handle, swing1LimitDegrees=15, swing2LimitDegrees=30, twistLimitDegrees=1}
 {bone1Name=dyn_chain2_handle_02, bone2Name=dyn_chain2_handle_01, swing1LimitDegrees=25, swing2LimitDegrees=15, twistLimitDegrees=2}
 {bone1Name=dyn_chain2_handle_03, bone2Name=dyn_chain2_handle_02, swing1LimitDegrees=20, swing2LimitDegrees=20, twistLimitDegrees=2}
 {bone1Name=dyn_chain2_handle_04, bone2Name=dyn_chain2_handle_03, swing1LimitDegrees=10, swing2LimitDegrees=10, twistLimitDegrees=1}
 {
  bone1Name=dyn_chain2_handle_04
  bone2Name=dyn_chain2_arm_05
  swing1LimitDegrees=45
  swing1Motion=Free
  swing2LimitDegrees=45
  swing2Motion=Free
  twistLimitDegrees=45
  twistMotion=Free
 }
]
