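# Animation Blueprint for the HonestWasp pickaxe that adds physics motion to its chain bones and toggles material elements when the wielder's body mesh is HonestWasp.
# Update reads pelvis, forearm and thigh transforms from the attach parent; matching bones get world-space replacement.
# The cached, constrained pose feeds two branches selected by bIsInFrontEnd.
# Front end runs four chained AnimDynamics nodes into RigidBody with its own physics asset; otherwise one RigidBody simulates alone.
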AnimBlueprint /BRCosmetics/Weapons/FORT_Melee/Pickaxe_HonestWasp/Meshes/HonestWasp_Axe_AnimBP
target_skeleton=&/HonestWasp_Axe_Skeleton
graphs [
 AnimGraph {nodes [
   n1 = RigidBody(ComponentPose=Local To Component(LocalPose=Use cached pose 'Base_CACHE'())) {
    bClampLinearTranslationLimitToRefPose=true
    bOverrideWorldGravity=true
    overridePhysicsAsset=&/HonestWasp_Axe_Physics
    simulationSpace=BaseBoneSpace
    baseBoneRef {boneName=Melee_TwoHanded_Handle}
    componentAppliedLinearAccClamp=<1000 1000 1000>
    overrideWorldGravity {z=-980}
    simSpaceSettings {
     dampingAlpha=0
     maxAngularAcceleration=40
     maxAngularVelocity=8
     maxLinearAcceleration=1300
     maxLinearVelocity=400
     worldAlpha=1
     externalLinearDragV=<0.3 0.3 0.3>
    }
   }
   n2 = Anim Dynamics - Bone: dyn_chain_01(ComponentPose=Local To Component(LocalPose=Use cached pose 'Base_CACHE'()), Alpha=1, ExternalForce="0,0,0") {
    angularDampingOverride=0.9
    angularSpringConstant=11000
    bOverrideAngularDamping=true
    bUseGravityOverride=true
    gravityScale=0.35
    simulationSpace=World
    boundBone {boneName=dyn_chain_01}
    componentAppliedLinearAccClamp=<9500 9500 9500>
    componentLinearAccScale=<0.5 0.5 0.5>
    componentLinearVelScale=<0.5 0.5 0.5>
    gravityOverride {z=-2200}
    physicsBodyDefinitions [{
      boundBone {boneName=dyn_chain_01}
      boxExtents=<15 20 20>
      constraintSetup {angularLimitsMax=<1 1 30>, angularLimitsMin=<-1 -1 -90>, angularTarget=<1 0 0>}
      localJointOffset=<1 0 0>
     }]
    relativeSpaceBone {boneName=b_AR_Root}
    simSpaceSettings {maxAngularAcceleration=35, maxAngularVelocity=0.8, simSpaceAngularAlpha=1}
   }
   n3 = Anim Dynamics - Bone: dyn_chain_02(ComponentPose=@n2, Alpha=1, ExternalForce="0,0,0") {
    angularDampingOverride=0.9
    angularSpringConstant=11000
    bOverrideAngularDamping=true
    bUseGravityOverride=true
    gravityScale=0.35
    simulationSpace=World
    boundBone {boneName=dyn_chain_02}
    componentAppliedLinearAccClamp=<9500 9500 9500>
    componentLinearAccScale=<0.5 0.5 0.5>
    componentLinearVelScale=<0.5 0.5 0.5>
    gravityOverride {z=-1800}
    physicsBodyDefinitions [{
      boundBone {boneName=dyn_chain_02}
      boxExtents=<15 20 20>
      constraintSetup {angularLimitsMax=<1 30 1>, angularLimitsMin=<-1 -30 -1>, angularTarget=<1 0 0>}
      localJointOffset=<2 0 0>
     }]
    relativeSpaceBone {boneName=b_AR_Root}
    simSpaceSettings {maxAngularAcceleration=35, maxAngularVelocity=0.8, simSpaceAngularAlpha=1}
   }
   n4 = Anim Dynamics - Bone: dyn_chain_03(ComponentPose=@n3, Alpha=1, ExternalForce="0,0,0") {
    angularDampingOverride=0.9
    angularSpringConstant=11000
    bOverrideAngularDamping=true
    bUseGravityOverride=true
    gravityScale=0.35
    simulationSpace=World
    boundBone {boneName=dyn_chain_03}
    componentAppliedLinearAccClamp=<9500 9500 9500>
    componentLinearAccScale=<0.5 0.5 0.5>
    componentLinearVelScale=<0.5 0.5 0.5>
    gravityOverride {z=-1600}
    physicsBodyDefinitions [{
      boundBone {boneName=dyn_chain_03}
      boxExtents=<15 20 20>
      constraintSetup {angularLimitsMax=<5 40 40>, angularLimitsMin=<-5 -40 -40>, angularTarget=<1 0 0>}
      localJointOffset=<3 0 0>
     }]
    relativeSpaceBone {boneName=b_AR_Root}
    simSpaceSettings {maxAngularAcceleration=35, maxAngularVelocity=0.7, simSpaceAngularAlpha=1}
   }
   n5 = Anim Dynamics - Bone: dyn_chain_04(ComponentPose=@n4, Alpha=1, ExternalForce="0,0,0") {
    angularDampingOverride=0.9
    angularSpringConstant=11000
    bOverrideAngularDamping=true
    bUseGravityOverride=true
    gravityScale=0.35
    simulationSpace=World
    boundBone {boneName=dyn_chain_04}
    componentAppliedLinearAccClamp=<9500 9500 9500>
    componentLinearAccScale=<0.5 0.5 0.5>
    componentLinearVelScale=<0.5 0.5 0.5>
    gravityOverride {z=-1600}
    physicsBodyDefinitions [{
      boundBone {boneName=dyn_chain_04}
      boxExtents=<15 20 20>
      constraintSetup {angularLimitsMax=<5 50 50>, angularLimitsMin=<-5 -50 -50>, angularTarget=<1 0 0>}
      localJointOffset=<4 0 0>
     }]
    relativeSpaceBone {boneName=b_AR_Root}
    simSpaceSettings {maxAngularAcceleration=35, maxAngularVelocity=0.7, simSpaceAngularAlpha=1}
   }
   n6 = RigidBody(ComponentPose=@n5, Alpha=0.8) {
    bClampLinearTranslationLimitToRefPose=true
    bOverrideWorldGravity=true
    overridePhysicsAsset=&/HonestWasp_FE_Axe_Physics
    simulationSpace=BaseBoneSpace
    baseBoneRef {boneName=Melee_TwoHanded_Handle}
    componentAppliedLinearAccClamp=<1000 1000 1000>
    overrideWorldGravity {z=-2000}
    simSpaceSettings {
     maxAngularAcceleration=40
     maxAngularVelocity=8
     maxLinearAcceleration=4000
     maxLinearVelocity=800
     worldAlpha=1
     externalLinearDragV=<0.3 0.3 0.1>
    }
   }
   Output Pose(Result=Blend Poses by bool(BlendPose_0=Component To Local(ComponentPose=@n6), BlendPose_1=Component To Local(ComponentPose=@n1), bActiveValue=Get bIsInFrontEnd())) {layerGroup=None}
   AnimGraphNode_SubInput_0 = Input Pose()
   n7 = "Transform (Modify) Bone - Bone: pelvis"(ComponentPose=Local To Component(LocalPose=@AnimGraphNode_SubInput_0), Alpha=1, Translation="0,0,0", Rotation="0,0,0", Scale="1,1,1") {
    rotationMode=BMM_Replace
    rotationSpace=BCS_WorldSpace
    translationMode=BMM_Replace
    translationSpace=BCS_WorldSpace
    boneToModify {boneName=pelvis}
   }
   n8 = "Transform (Modify) Bone - Bone: lowerarm_r"(ComponentPose=@n7, Alpha=1, Translation="0,0,0", Rotation="0,0,0", Scale="1,1,1") {
    rotationMode=BMM_Replace
    rotationSpace=BCS_WorldSpace
    translationMode=BMM_Replace
    translationSpace=BCS_WorldSpace
    boneToModify {boneName=lowerarm_r}
   }
   n9 = "Transform (Modify) Bone - Bone: thigh_r"(ComponentPose=@n8, Alpha=1, Translation="0,0,0", Rotation="0,0,0", Scale="1,1,1") {
    rotationMode=BMM_Replace
    rotationSpace=BCS_WorldSpace
    translationMode=BMM_Replace
    translationSpace=BCS_WorldSpace
    boneToModify {boneName=thigh_r}
   }
   n10 = "Constraint - base_chain2_arm (T)"(ComponentPose=@n9, Alpha=1, ConstraintWeights_0=1) {
    boneToModify {boneName=base_chain2_arm}
    constraintSetup [{offsetOption=None, transformType=Translation, targetBone {boneName=base_chain_arm}}]
    constraintWeights [1]
   }
   Save cached pose 'Base_CACHE'(Pose=Component To Local(ComponentPose=@n10)) {cacheName=Base_CACHE}
  ]}
 EventGraph {nodes [
   Event BlueprintUpdateAnimation()
   > n1 = Sequence!(execute) -> (then_0, then_1, then_2:Exec)
   n2 = Get Dynamic Material Instance(execute=@n1.then_1, SkeletalMeshComponent=GetPartSkeletalMeshComponent())
   > n3 = Branch(Condition=Get HonestWasp_Body() -> (HonestWasp_Body:bool))
   > SetScalarParameterValue(self=@n2, ParameterName=Hide Element 01, Value=0)
   > SetScalarParameterValue(self=@n2, ParameterName=Hide Element 1_5, Value=1)
   SetScalarParameterValue(execute=@n3.else, self=@n2, ParameterName=Hide Element 01, Value=1)
   > SetScalarParameterValue(self=@n2, ParameterName=Hide Element 1_5, Value=0)
   n4 = Cast To FortPlayerPawn(execute=@n1.then_0, Object=TryGetPawnOwner())
   n5 = GetSkeletalMeshForPartType(self=@n4)
   > Set HonestWasp_Body(HonestWasp_Body:bool=MatchesWildcard(SourceString=GetDisplayName(Object=@n5), Wildcard=*M_MED_HonestWasp*, SearchCase=CaseSensitive)) -> (Output_Get:bool)
   n6 = GetPartSkeletalMeshComponent()
   Is Valid(exec=@n1.then_2, InputObject=@n6)
   n7 = GetAttachParent!(self:Scene Component Object Reference=@n6) -> (ReturnValue:Scene Component Object Reference)
   > Is Valid(InputObject=@n7)
   > n8 = Set Body_Mesh(Body_Mesh:Scene Component Object Reference=@n7) -> (Output_Get:Scene Component Object Reference)
   n9 = BreakTransform(InTransform=GetSocketTransform(self=@n8, InSocketName=pelvis, TransformSpace=RTS_World))
   > Set PelvisLocation(PelvisLocation:Vector=@n9) -> (Output_Get:Vector)
   > Set Pelvis_Rotation(Pelvis_Rotation:Rotator=@n9.Rotation) -> (Output_Get:Rotator)
   n10 = BreakTransform(InTransform=GetSocketTransform(self=@n8, InSocketName=lowerarm_r, TransformSpace=RTS_World))
   > Set LowerArm_R_Location(LowerArm_R_Location:Vector=@n10) -> (Output_Get:Vector)
   > Set LowerArm_R_Rotation(LowerArm_R_Rotation:Rotator=@n10.Rotation) -> (Output_Get:Rotator)
   n11 = BreakTransform(InTransform=GetSocketTransform(self=@n8, InSocketName=thigh_r, TransformSpace=RTS_World))
   > Set Thigh_R_Location(Thigh_R_Location:Vector=@n11) -> (Output_Get:Vector)
   > Set Thigh_R_Rotation(Thigh_R_Rotation:Rotator=@n11.Rotation) -> (Output_Get:Rotator)
  ]}
]
variables [
 Body_Mesh {type=SceneComponent, default=null}
 FE_WithPawn {type=boolean}
 HonestWasp_Body {type=boolean}
 HonestWasp_Pickaxe {type=SkeletalMesh, default=&/HonestWasp_Axe}
 LowerArm_R_Location {type=Vector}
 LowerArm_R_Rotation {type=Rotator}
 Pelvis_Rotation {type=Rotator}
 PelvisLocation {type=Vector}
 Thigh_R_Location {type=Vector}
 Thigh_R_Rotation {type=Rotator}
]
